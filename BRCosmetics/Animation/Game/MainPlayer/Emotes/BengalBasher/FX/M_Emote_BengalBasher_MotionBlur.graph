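# Translucent, unlit, two-sided motion blur material for Niagara mesh particles in the Bengal Basher emote.
# A smoke-trail texture is scaled, rotated a quarter turn and panned, with the UV parameter's U channels driving both axes.
# A V-gradient power mask with a blendable bottom fade shapes the streak, tuned by dynamic Power, Intensity, BottomPower and BottomBlend.
# Particle alpha multiplies that result, while particle RGB passes through an emissive adjustment function.
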Material /BRCosmetics/Animation/Game/MainPlayer/Emotes/BengalBasher/FX/M_Emote_BengalBasher_MotionBlur
graph {
 properties {
  bUsedWithNiagaraMeshParticles=true
  blendMode=BLEND_Translucent
  pixelDepthOffsetMode=PDOM_Legacy
  shadingModel=MSM_Unlit
  twoSided=true
  usedShadingModels=Unlit
 }
 nodes [
  n1 = ParticleColor()
  MaterialFunctionCall(Emissive=@n1.RGB) {materialFunction=&MF_FX_EmissiveAdjustments}
  n2 = VectorParameter() {parameterName=UV, channelNames {a=V_Speed, b=U_Speed, g=V_Scale, r=U_Scale}, defaultValue=<rgba 1 1 0 -20>}
  n3 = Multiply(A=TextureCoordinate(), B=AppendVector(A=@n2.U_Scale, B=@n2.U_Scale))
  n4 = Constant() {r=0.25}
  n5 = TextureSampleParameter2D(UVs=Panner(
    Coordinate=MaterialFunctionCall(UVs=@n3, "Rotation Angle (0-1)"=@n4) {materialFunction=&CustomRotator}.Rotated Values
    Speed=AppendVector(A=@n2.U_Speed, B=@n2.U_Speed)
   ) {speedY=-20}) {parameterName=Texture, texture=/BRCosmetics/Weapons/FORT_Melee/Pickaxe_Llama_Hero_Winter_Male/FX/T_Puffy_Smoke_Trail_01}
  n6 = ComponentMask(None=TextureCoordinate()) {g=true}
  n7 = DynamicParameter() {defaultValue=<rgba 1 1 1 0>, paramNames [Power,Intensity,BottomPower,BottomBlend]}
  n8 = OneMinus(None=Power(Base=@n6, Exp=@n7.BottomPower) {constExponent=50})
  n9 = Saturate(None=Multiply(A=Power(Base=@n6, Exp=@n7.Power) {constExponent=50}, B=LinearInterpolate(B=@n8, Alpha=@n7.BottomBlend) {constA=1}))
  Multiply(A=@n1.A, B=Multiply(A=Multiply(A=@n5.RGB, B=@n9), B=@n7.Intensity) {constB=20})
 ]
}
