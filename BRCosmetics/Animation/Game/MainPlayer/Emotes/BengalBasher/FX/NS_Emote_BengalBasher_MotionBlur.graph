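# Niagara effect that supplies mesh-based motion blur for the BengalBasher player emote.
# Two emitters, MotionBlur and MotionBlurEdge, each burst-spawn an identically scaled and oriented gray mesh particle that is not killed when its lifetime elapses.
# The edge layer uses higher alpha and writes dynamic material parameters.
# Both scale particle color through a time-of-day scalar, dimming at night, morning and evening.
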NiagaraSystem /BRCosmetics/Animation/Game/MainPlayer/Emotes/BengalBasher/FX/NS_Emote_BengalBasher_MotionBlur
dependencies {usedDataInterfaces [], usedDynamicInputs [], usedModules [&SystemState]}
emitters {
 MotionBlur {
  emitterUpdateScript [
   &EmitterState()
   &SpawnBurst_Instantaneous(Use Spawn Probability=false, UseLoopCountLimit=false)
  ]
  particleSpawnScript [
   &InitializeParticle(
    Color=<rgba 0.1667 0.1649 0.1649 0.5>
    Lifetime=0.62
    Mesh Scale Mode=::NewEnumerator3
    Mesh Uniform Scale=0.515
    UsePositionOffset=false
    Write Material Random=true
   )
   &InitialMeshOrientation(Mesh Orientation Mode=::NewEnumerator1, Rotation=<0.25 0 0>, Use Rotation Vector=true)
  ]
  particleUpdateScript [
   &ParticleState(Kill Particles When Lifetime Has Elapsed=false)
   &ScaleColor(
    Scale RGB=&VectorFromFloat(Value=&FNTimeOfDayScalar(Night Scale=0.5, Morning Scale=0.8, Evening Scale=0.7))
    ScaleA=true
    ScaleRGB=true
   )
  ]
  rendererClasses [&NiagaraMeshRendererProperties]
  renderers [{rendererClass=&NiagaraMeshRendererProperties, rendererIndex=0}]
 }
 MotionBlurEdge {
  emitterUpdateScript [
   &EmitterState()
   &SpawnBurst_Instantaneous(Use Spawn Probability=false, UseLoopCountLimit=false)
  ]
  particleSpawnScript [
   &InitializeParticle(
    Color=<rgba 0.1667 0.1649 0.1649 1>
    Lifetime=0.62
    Mesh Scale Mode=::NewEnumerator3
    Mesh Uniform Scale=0.515
    UsePositionOffset=false
    Write Material Random=true
   )
   &InitialMeshOrientation(Mesh Orientation Mode=::NewEnumerator1, Rotation=<0.25 0 0>, Use Rotation Vector=true)
  ]
  particleUpdateScript [
   &ParticleState(Kill Particles When Lifetime Has Elapsed=false)
   &ScaleColor(
    Scale RGB=&VectorFromFloat(Value=&FNTimeOfDayScalar(Night Scale=0.5, Morning Scale=0.8, Evening Scale=0.7))
    ScaleA=true
    ScaleRGB=true
   )
   &DynamicMaterialParameters(
    Index 0 Param 1=50
    Index 0 Param 2=2
    Module.Index0Param0Write=true
    Module.Index0Param1Write=true
    Module.Index0Param2Write=true
    Module.Index0Param3Write=true
   )
  ]
  rendererClasses [&NiagaraMeshRendererProperties]
  renderers [{rendererClass=&NiagaraMeshRendererProperties, rendererIndex=0}]
 }
}
summary {
 systemName=NS_Emote_BengalBasher_MotionBlur
 emitters [
  {emitterName=MotionBlur, rendererClasses [&NiagaraMeshRendererProperties]}
  {emitterName=MotionBlurEdge, rendererClasses [&NiagaraMeshRendererProperties]}
 ]
 userVariables [
  User.bIsFrontEnd {type=NiagaraBool}
  User.bIsFrontEndPreview {type=NiagaraBool}
 ]
}
