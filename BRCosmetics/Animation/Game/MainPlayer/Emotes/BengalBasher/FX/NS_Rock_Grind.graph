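# Niagara spark effect for the BengalBasher emote, spraying short-lived, glowing orange sprites for a rock grind.
# A single sprite emitter spawns at locations sampled from a skeletal mesh, with its rate driven by User.SparksSpawn_01.
# Sparks launch in a cone under randomized world and local gravity, with User.SparksVelocity bounding a random velocity scale.
# Sprites stretch with speed, then shrink and fade from HDR color over their life.
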NiagaraSystem /BRCosmetics/Animation/Game/MainPlayer/Emotes/BengalBasher/FX/NS_Rock_Grind
dependencies {
 usedDataInterfaces [
  &NiagaraDataInterfaceCurve
  &NiagaraDataInterfaceMeshRendererInfo
  &NiagaraDataInterfaceArrayFloat3
  &NiagaraDataInterfaceVector2DCurve
  &NiagaraDataInterfaceColorCurve
  &NiagaraDataInterfaceSkeletalMesh
 ]
 usedDynamicInputs []
 usedModules [
  &SystemState
  &self:SystemSpawnScript.SystemScriptSource.SystemScriptGraph.NiagaraNodeAssignment_5.SetVariables_#1
 ]
}
emitters {GlowDrops_CPU_Local002 {
  emitterUpdateScript [
   &EmitterState(
    Enable Distance Culling=true
    Life Cycle Mode=::NewEnumerator1
    Loop Behavior=::NewEnumerator1
    Loop Duration=1.75
    Module.bUseMaxDistance=true
    Module.bUseMinDistance=false
    Scalability Mode=::NewEnumerator1
    UseLoopDelay=true
   )
   &SpawnRate(
    SpawnRate=&FloatFromCurve(
     FloatCurve=DataInterface() {propertyValues {Curve {keys [{arriveTangent=-43.9813, interpMode=RCIM_Cubic, leaveTangent=-43.9813, tangentMode=RCTM_User, time=0, value=20}, <at 1 0>]}}}
     CurveIndex=$Emitter.NormalizedLoopAge
     Scale Curve=4
    )
    Use Spawn Probability=false
   ) {enabled=false}
   &SpawnRate(SpawnRate=&Multiply_Float(A=$User.SparksSpawn_01, B=7), Use Spawn Probability=false) {module=SpawnRate001}
  ]
  particleSpawnScript [
   &FN_SetSpawnGroup(Number of Spawn Groups=1)
   &InitializeParticle(
    Color Mode=::NewEnumerator0
    Lifetime Max=0.1
    Lifetime Min=0.05
    Lifetime Mode=::NewEnumerator1
    Sprite Size Max=&Multiply_Vector2D(A=<3 10>, B=&Vector2DFromFloat(Value=1.1))
    Sprite Size Min=&Multiply_Vector2D(A=<2 5>, B=&Vector2DFromFloat(Value=1.1))
    Sprite Size Mode=::NewEnumerator2
    UsePositionOffset=false
    Write Material Random=true
   )
   &SkeletalMeshLocation(
    MaskSpawnGroup=false
    Mesh Sampling Type=::NewEnumerator5
    Position Sampling=::NewEnumerator2
    Sampled Position Offset=<3 -7 -4>
    Skeletal Mesh=$System.Skeletal Mesh
    Socket Sampling Mode=::NewEnumerator5
   ) {enabled=false}
   &SkeletalMeshLocation(
    MaskSpawnGroup=false
    Mesh Sampling Type=::NewEnumerator5
    Position Sampling=::NewEnumerator2
    Skeletal Mesh=$System.Skeletal Mesh
    Socket Sampling Mode=::NewEnumerator5
   ) {module=SkeletalMeshLocation001}
   &ShapeLocation(MaskSpawnGroup=false, Non Uniform Scale=<0 0 0>, Override Seed=false, UseOwnerScale=true) {enabled=false}
   &AddVelocity(
    Cone Angle=&RandomRangeFloat(Minimum=10, Maximum=20)
    Inner Cone Angle=5
    Module.UseConeFalloff=false
    Module.UseVelDistribution=false
    Override Seed=false
    Rotation Angle=0.05
    Rotation Angle Type:ENiagara_AngleInput=::NewEnumerator1
    Rotation Axis:Vector3f=<0 1 0>
    Rotation Mode=::NewEnumerator1
    Velocity Mode=::NewEnumerator2
    Velocity Speed=&RandomRangeFloat(Minimum=500, Maximum=750)
   )
  ]
  particleUpdateScript [
   &ParticleState()
   &GravityForce(Gravity=&RandomRangeVector(Minimum=<0 0 -920>, Maximum=<0 0 -480>))
   &GravityForce(Coordinate Space=::Local, Gravity=&RandomRangeVector(Minimum=<0 0 0>, Maximum=<-750 0 0>)) {module=GravityForce001}
   &AerodynamicDrag(
    Aerodynamic Drag:NiagaraFloat=&RandomRangeFloat(Minimum=0.2, Maximum=1)
    Aerodynamic Rotational Drag=0.2
    Surface Normal Axis:Vector3f=&SelectVectorFromArray(Direct Array Index=$Particles.SpawnGroup, Vector Selection Array=DataInterface() {propertyValues {FloatData [<0.2 -0.2 1>, <0.2 -0.2 1>]}})
    Use Linear Drag=true
    Use Rotational Drag=true
   ) {enabled=false}
   &ScaleVelocity(Velocity Scale=&RandomRangeVector(
     Minimum=<0 0 0>
     Maximum=&VectorFromFloat(Value=&Multiply_Float(A=0.75, B=$User.SparksVelocity))
    ))
   &ScaleVelocity(Velocity Scale=&Multiply_Vector(
     A=$Particles.Initial.SkeletalMeshLocation.SampledVelocity
     B=&RandomRangeVector(Minimum=<0 0 0>, Maximum=<0.1 0.1 0.1>)
    )) {enabled=false, module=ScaleVelocity001}
   &SolveForcesAndVelocity(Clamp Velocity=false, Limit Acceleration=false)
   &ScaleSpriteSize(Scale Factor=&Vector2DFromCurve(
     Vector2Curve=DataInterface() {propertyValues {
       XCurve {keys [<at 0 1 RCIM_Cubic>, {arriveTangent=-1.2073, interpMode=RCIM_Cubic, leaveTangent=-1.2073, tangentMode=RCTM_User, time=1, value=0.55}]}
       YCurve {keys [<at 0 1 RCIM_Cubic>, {arriveTangent=-1.2073, interpMode=RCIM_Cubic, leaveTangent=-1.2073, tangentMode=RCTM_User, time=1, value=0.55}]}
      }}
     ScaleCurve=<1 1>
    ))
   &ScaleSpriteSizeBySpeed(Max Scale Factor=<0.5 2.5>, Min Scale Factor=<0 0>)
   &AlignSpriteToMeshOrientation(SetAlignment=true, SetFacing=true)
   &Color(Color=&ColorFromCurve(ColorCurve=DataInterface() {propertyValues {
       AlphaCurve {keys [<at 0 1 RCIM_Cubic>,<at 1 0 RCIM_Cubic>]}
       BlueCurve {keys [<at 0 0 RCIM_Cubic>,<at 1 0 RCIM_Cubic>]}
       GreenCurve {keys [<at 0 25 RCIM_Cubic>,<at 1 0.4167 RCIM_Cubic>]}
       RedCurve {keys [<at 0 75 RCIM_Cubic>,<at 1 25 RCIM_Cubic>]}
      }}))
  ]
  rendererClasses [&NiagaraSpriteRendererProperties]
  renderers [{rendererClass=&NiagaraSpriteRendererProperties}]
 }}
summary {
 systemName=NS_Rock_Grind
 emitters [{emitterName=GlowDrops_CPU_Local002, rendererClasses [&NiagaraSpriteRendererProperties]}]
 userVariables [
  User.bIsFrontEnd {type=NiagaraBool}
  User.bIsFrontEndPreview {type=NiagaraBool}
  User.Delay {type=NiagaraFloat, defaultValue=0, description=""}
  User.Jumped {type=NiagaraBool}
  User.Leaning Direction {type=NiagaraFloat, defaultValue=0, description=""}
  User.NormalizedNotifyProgress {type=NiagaraFloat, defaultValue=0, description=""}
  User.SparksSpawn_01 {type=NiagaraFloat, defaultValue=0, description=""}
  User.SparksVelocity {type=NiagaraFloat, defaultValue=0, description=""}
 ]
}
